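; Camera mode Blueprint for Umbramolt Kipper's tactical sprint, judging by its name and camera folder.
; It overrides GetCameraLookAtLocation but only calls the parent function and returns its look-at location and return value unchanged.
; The file shows no other logic, so look-at behaviour comes entirely from the parent camera mode.
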(fn GetCameraLookAtLocation
  (bind (_outlookatlocation _returnvalue) (|CallParentFunction))
  (return _outlookatlocation _returnvalue))
